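# UI sound cue for front-end close feedback, named for closing the widget offer to show the lobby.
# A mixer layers a main button click with two copies of the party-invite close sound.
# Each close layer has its own modulator for volume and pitch, one pitched down and the other slightly randomized.
# It uses the Fort_UI sound class and a front-end menu concurrency set named for stop-oldest behavior.
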SoundCue /Game/Sounds/Fort_UI_Sounds/FrontEnd/New_FrontEnd/FN_UI_Close_WidgetOfferToShowLobby_Cue
root=Mixer_0
nodes [
 {id=Mixer_0, children [WavePlayer_1,Modulator_1,Modulator_2], properties {inputVolume [1,1.5,1.5]}}
 {id=Modulator_1, children [WavePlayer_0], properties {pitchMax=0.9, pitchMin=0.9, volumeMax=0.3, volumeMin=0.25}}
 {id=Modulator_2, children [WavePlayer_2], properties {pitchMax=1, pitchMin=0.9, volumeMax=0.6, volumeMin=0.6}}
 {id=WavePlayer_0, properties {soundWaveAssetPtr=/Game/Sounds/Fort_UI_Sounds/FrontEnd/ItemShop/wav/UI_PartyInvite_Close}}
 {id=WavePlayer_1, properties {soundWaveAssetPtr=&/UI_FN_Button_Click_Main_01}}
 {id=WavePlayer_2, properties {soundWaveAssetPtr=&UI_PartyInvite_Close}}
]
properties {
 maxDistance=2097152
 soundClassObject=&Fort_UI
 volumeMultiplier=1
 concurrencyOverrides {maxCount=1, resolutionRule=StopOldest}
 concurrencySet [/Game/Sounds/Concurrency_Rules/FrontEnd_UI_Menu_Max_03_Stop_Oldest]
}
