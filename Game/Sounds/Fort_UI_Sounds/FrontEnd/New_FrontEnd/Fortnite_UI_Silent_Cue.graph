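# Silent sound cue for the Fortnite front-end UI that produces no audible sound when played.
# A Wave Player feeds a one-second silence wave into an Enveloper root with a short fade-out on volume.
# It uses the Fort_UI sound class and a shared front-end menu concurrency set, named for a three-voice, stop-oldest limit.
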SoundCue /Game/Sounds/Fort_UI_Sounds/FrontEnd/New_FrontEnd/Fortnite_UI_Silent_Cue
root=Enveloper_0
nodes [
 {
  id=Enveloper_0
  children [WavePlayer_2]
  properties {pitchCurve {editorCurveData {keys [{time=0}]}}, volumeCurve {editorCurveData {keys [<at 0 1>,<at 0.1 0>]}}}
 }
 {id=WavePlayer_2, properties {soundWaveAssetPtr=&/1second_silence}}
]
properties {
 maxDistance=2097152
 soundClassObject=&Fort_UI
 volumeMultiplier=1.2
 concurrencyOverrides {maxCount=1, resolutionRule=StopOldest}
 concurrencySet [/Game/Sounds/Concurrency_Rules/FrontEnd_UI_Menu_Max_03_Stop_Oldest]
}
