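# UI sound cue for hovering on the theater map in the front end, routed through the Fort_UI sound class.
# A mixer layers three waves at separate levels: a slightly delayed map-hover sound, plus generic button mouse-over and hover-click sounds pitched down through modulators with fixed values.
# A concurrency override allows one instance at a time and stops the oldest, so rapid hovers do not stack.
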SoundCue /Game/Sounds/Fort_UI_Sounds/FrontEnd/New_FrontEnd/Fortnite_UI_Theater_Map_Hover_01_Cue
root=Mixer_0
nodes [
 {id=Delay_0, children [WavePlayer_4], properties {delayMax=0.025, delayMin=0.025}}
 {id=Mixer_0, children [Delay_0,Modulator_1,Modulator_0], properties {inputVolume [0.75,0.25,1]}}
 {id=Modulator_0, children [WavePlayer_2], properties {pitchMax=0.75, pitchMin=0.75, volumeMax=1, volumeMin=1}}
 {id=Modulator_1, children [WavePlayer_1], properties {pitchMax=0.75, pitchMin=0.75, volumeMax=1, volumeMin=1}}
 {id=WavePlayer_1, properties {soundWaveAssetPtr=&/UI_FN_Button_MoueOver}}
 {id=WavePlayer_2, properties {soundWaveAssetPtr=&/UI_FN_Button_Hover_Click_01}}
 {id=WavePlayer_4, properties {soundWaveAssetPtr=&/Fortnite_UI_Map_Hover_01}}
]
properties {
 bOverrideConcurrency=true
 maxDistance=2097152
 soundClassObject=&Fort_UI
 concurrencyOverrides {maxCount=1, resolutionRule=StopOldest}
}
