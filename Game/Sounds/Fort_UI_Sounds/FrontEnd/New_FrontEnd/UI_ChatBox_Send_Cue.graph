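# UI sound cue for sending a chat box message in the front end, routed to the Fort_UI sound class.
# A mixer layers a high-pitched button hover/click, a briefly delayed chatbox send sound, and a quiet, pitched-down button press.
# Modulators set pitch and volume per layer, adding slight random variation.
# A concurrency override limits simultaneous copies and blocks new plays once the limit is reached.
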SoundCue /Game/Sounds/Fort_UI_Sounds/FrontEnd/New_FrontEnd/UI_ChatBox_Send_Cue
root=Mixer_0
nodes [
 {id=Delay_0, children [Modulator_3], properties {delayMax=0.1, delayMin=0.1}}
 {id=Mixer_0, children [Modulator_1,Delay_0,Modulator_2], properties {inputVolume [1,1,1]}}
 {id=Modulator_1, children [WavePlayer_2], properties {pitchMax=1.65, pitchMin=1.6, volumeMax=1, volumeMin=1}}
 {id=Modulator_2, children [WavePlayer_3], properties {pitchMax=0.5, pitchMin=0.5, volumeMax=0.23, volumeMin=0.15}}
 {id=Modulator_3, children [WavePlayer_0], properties {pitchMin=1.1, volumeMax=0.6, volumeMin=0.6}}
 {id=WavePlayer_0, properties {soundWaveAssetPtr=&/UI_Chatbox_Send}}
 {id=WavePlayer_2, properties {soundWaveAssetPtr=&/UI_FN_Button_Hover_Click_01}}
 {
  id=WavePlayer_3
  properties {soundWaveAssetPtr=/Game/Sounds/Fort_UI_Sounds/FrontEnd/DefaultSounds/HoldableButtons/wav/UI_Button_Press_Hold_Quiet_Press}
 }
]
properties {
 bOverrideConcurrency=true
 maxDistance=2097152
 soundClassObject=&Fort_UI
 volumeMultiplier=1
 concurrencyOverrides {maxCount=3, resolutionRule=PreventNew, retriggerTime=0.2}
}
