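# Sound cue for the chat text notification in the front-end UI.
# A single wave player feeds the chatbox notification wave straight to the output, with a slight volume boost and the application tray notification sound class.
# Its own concurrency override allows one instance at a time and rejects new plays, with a short retrigger time to limit rapid repeats.
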SoundCue /Game/Sounds/Fort_UI_Sounds/FrontEnd/New_FrontEnd/UI_Chat_Text_Notification_Cue
root=WavePlayer_1
nodes [{id=WavePlayer_1, properties {soundWaveAssetPtr=&/UI_Chatbox_Notification}}]
properties {
 bOverrideConcurrency=true
 maxDistance=2097152
 soundClassObject=/Game/Sounds/FortSoundClassesAndModes/Fort_Application_TrayNotification
 volumeMultiplier=1.06
 concurrencyOverrides {maxCount=1, resolutionRule=PreventNew, retriggerTime=0.4}
}
