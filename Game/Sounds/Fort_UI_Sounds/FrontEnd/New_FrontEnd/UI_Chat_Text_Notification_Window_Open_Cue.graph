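# UI sound cue for the front end that, by its name, plays when a chat text notification window opens.
# A single wave player for the chatbox open sound feeds a default-setting Modulator, with cue-level pitch raised and volume lowered.
# It routes through the Fort UI sound class and overrides concurrency to cap simultaneous plays, blocking new instances at the limit, with a short retrigger time.
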SoundCue /Game/Sounds/Fort_UI_Sounds/FrontEnd/New_FrontEnd/UI_Chat_Text_Notification_Window_Open_Cue
root=Modulator_0
nodes [
 {id=Modulator_0, children [WavePlayer_1]}
 {id=WavePlayer_1, properties {soundWaveAssetPtr=&/UI_Chatbox_Open}}
]
properties {
 bOverrideConcurrency=true
 maxDistance=2097152
 pitchMultiplier=1.2
 soundClassObject=&Fort_UI
 volumeMultiplier=0.8
 concurrencyOverrides {maxCount=3, resolutionRule=PreventNew, retriggerTime=0.2}
}
