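# UI sound cue for the Armory tab pop-ups in the front end, judging by its name.
# A single wave player plays one sound wave with boosted volume in the Fort_UI sound class.
# A concurrency override allows one instance at a time and prevents new plays while it is active.
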SoundCue /Game/Sounds/Fort_UI_Sounds/FrontEnd/New_FrontEnd/UI_FN_Armory_Tab_PopUps_Cue
root=WavePlayer_1
nodes [{id=WavePlayer_1, properties {soundWaveAssetPtr=&/UI_FN_Armory_Tab_PopUps_01}}]
properties {
 bOverrideConcurrency=true
 maxDistance=2097152
 soundClassObject=&Fort_UI
 volumeMultiplier=1.5
 concurrencyOverrides {maxCount=1, resolutionRule=PreventNew}
}
