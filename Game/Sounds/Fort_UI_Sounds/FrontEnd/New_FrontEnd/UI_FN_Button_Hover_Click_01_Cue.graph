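# Sound cue for a front-end UI button hover click.
# A single wave player feeds the UI_FN_Button_Hover_Click_01 wave straight to the output, with no randomization or modulation, at a boosted volume on the Fort_UI sound class.
# A shared front-end menu concurrency rule, named for a one-voice limit that stops the oldest hover tick, governs overlapping plays.
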SoundCue /Game/Sounds/Fort_UI_Sounds/FrontEnd/New_FrontEnd/UI_FN_Button_Hover_Click_01_Cue
root=WavePlayer_0
nodes [{id=WavePlayer_0, properties {soundWaveAssetPtr=&/UI_FN_Button_Hover_Click_01}}]
properties {
 maxDistance=2097152
 soundClassObject=&Fort_UI
 volumeMultiplier=1.4
 concurrencySet [/Game/Sounds/Concurrency_Rules/FrontEnd_UI_Menu_Max_01_Stop_Oldest_Hover_Tick]
}
