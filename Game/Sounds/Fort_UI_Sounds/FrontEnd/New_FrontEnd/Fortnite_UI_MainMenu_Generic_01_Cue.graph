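# Sound cue for a generic main menu button click in the Fortnite front-end UI.
# A single wave player feeds the click wave straight to the output, with no randomization or modulation.
# It uses the Fort_UI sound class, disables virtualization and references the shared pressed-button concurrency asset.
# It also carries concurrency override values for a single stop-oldest voice with a short retrigger time.
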SoundCue /Game/Sounds/Fort_UI_Sounds/FrontEnd/New_FrontEnd/Fortnite_UI_MainMenu_Generic_01_Cue
root=WavePlayer_1
nodes [{id=WavePlayer_1, properties {soundWaveAssetPtr=&/UI_FN_Button_Click_Main_01}}]
properties {
 maxDistance=2097152
 soundClassObject=&Fort_UI
 virtualizationMode=Disabled
 volumeMultiplier=1
 concurrencyOverrides {maxCount=1, resolutionRule=StopOldest, retriggerTime=0.15}
 concurrencySet [/Game/Sounds/Fort_UI_Sounds/FrontEnd/DefaultSounds/PressedButtons/UI_PressedButton_Concurrency]
}
