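# Sound Cue for a generic front-end main menu UI sound, the third numbered variant, routed to the Fort_UI sound class.
# A single Wave Player feeds the root directly at reduced volume, with no randomization, modulation or mixing nodes.
# An overridden concurrency limits simultaneous instances, stops the oldest when the limit is hit, and sets a short retrigger time and volume scaling to restrain rapid repeats.
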SoundCue /Game/Sounds/Fort_UI_Sounds/FrontEnd/New_FrontEnd/Fortnite_UI_MainMenu_Generic_03_Cue
root=WavePlayer_0
nodes [{id=WavePlayer_0, properties {soundWaveAssetPtr=&/Fortnite_UI_MainMenu_Generic_03}}]
properties {
 bOverrideConcurrency=true
 maxDistance=2097152
 soundClassObject=&Fort_UI
 volumeMultiplier=0.5
 concurrencyOverrides {maxCount=2, resolutionRule=StopOldest, retriggerTime=0.15, volumeScale=0.5}
}
