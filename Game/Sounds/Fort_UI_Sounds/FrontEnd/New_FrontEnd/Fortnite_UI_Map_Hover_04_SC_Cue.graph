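# Sound cue for a front-end UI map hover sound, routed to the Fort_UI sound class.
# A single wave player plays the default MP hover wave, with no randomisation or modulation.
# A shared hover-tick concurrency rule is assigned, alongside override values that specify one voice and stopping the oldest.
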SoundCue /Game/Sounds/Fort_UI_Sounds/FrontEnd/New_FrontEnd/Fortnite_UI_Map_Hover_04_SC_Cue
root=WavePlayer_3
nodes [{id=WavePlayer_3, properties {soundWaveAssetPtr=/Game/Sounds/Fort_UI_Sounds/FrontEnd/DefaultSounds/PressedButtons/wav/UI_MP_Hover_Default_01}}]
properties {
 maxDistance=2097152
 soundClassObject=&Fort_UI
 volumeMultiplier=1
 concurrencyOverrides {maxCount=1, resolutionRule=StopOldest}
 concurrencySet [/Game/Sounds/Concurrency_Rules/FrontEnd_UI_Menu_Max_01_Stop_Oldest_Hover_Tick]
}
