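# Sound cue for the front-end UI mission objective sound, routed to the Fort_UI sound class.
# A mixer layers a main button click with a mission objective pop-up, with the pop-up boosted relative to the click and routed through a delay node left at default settings.
# A concurrency override limits the cue to one instance and blocks new plays, taking precedence over the assigned shared concurrency set.
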SoundCue /Game/Sounds/Fort_UI_Sounds/FrontEnd/New_FrontEnd/Fortnite_UI_MissionObj_Cue
root=Mixer_0
nodes [
 {id=Delay_0, children [WavePlayer_3]}
 {id=Mixer_0, children [WavePlayer_1,Delay_0], properties {inputVolume [1,2]}}
 {id=WavePlayer_1, properties {soundWaveAssetPtr=&/UI_FN_Button_Click_Main_01}}
 {id=WavePlayer_3, properties {soundWaveAssetPtr=&/UI_FN_Mission_Obj_PopUp_01}}
]
properties {
 bOverrideConcurrency=true
 maxDistance=2097152
 soundClassObject=&Fort_UI
 volumeMultiplier=1
 concurrencyOverrides {maxCount=1, resolutionRule=PreventNew}
 concurrencySet [/Game/Sounds/Concurrency_Rules/ThreeMaximum_Owner_PreventNew]
}
