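# UI_BlankForOverride_Cue is a silent front-end UI sound cue, named as a blank for overriding other sounds.
# A single wave player root references the Fort_UI_ComboUo_04 gameplay wave, with no other nodes in the graph.
# The cue's volume multiplier is zero, so that wave is inaudible, and the cue is assigned to the Fort_UI sound class.
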SoundCue /Game/Sounds/Fort_UI_Sounds/FrontEnd/New_FrontEnd/UI_BlankForOverride_Cue
root=WavePlayer_1
nodes [{id=WavePlayer_1, properties {soundWaveAssetPtr=/Game/Sounds/Fort_UI_Sounds/Gameplay/Fort_UI_ComboUo_04}}]
properties {maxDistance=2097152, soundClassObject=&Fort_UI, volumeMultiplier=0}
